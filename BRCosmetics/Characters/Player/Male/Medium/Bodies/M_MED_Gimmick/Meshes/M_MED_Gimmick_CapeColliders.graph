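# Physics asset of cape colliders for the medium male Gimmick character body, giving collision volumes for the cape to rest against.
# Capsules cover pelvis, spine, neck, clavicles and both dynamic belt chains, with a sphere on the head and a large flat box on the root as a likely ground plane.
# Limb bodies for arms, legs and feet carry no shapes, and no constraints link the bodies, so it serves collision rather than ragdoll.
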PhysicsAsset /BRCosmetics/Characters/Player/Male/Medium/Bodies/M_MED_Gimmick/Meshes/M_MED_Gimmick_CapeColliders
bodies {
 calf_l []
 calf_r []
 clavicle_l [{shapeType=Capsule, length=7.7636, radius=11.1427, center=<9.9642 -0.7741 -1.5923>, rotation=<rot 75.9352 -0.0001 175.5622>}]
 clavicle_r [{shapeType=Capsule, length=7.7636, radius=11.1427, center=<-9.9642 0.7741 1.5923>, rotation=<rot 75.9351 -0.0001 -4.4378>}]
 dyn_belt_l_01 [{shapeType=Capsule, length=11.8293, radius=4.6234, shapeName=dyn_belt_r_01_capsule, center=<4 -0.0004 4.1254>, rotation=<rot 90 0 0>}]
 dyn_belt_l_02 [{shapeType=Capsule, length=11.8293, radius=4.6234, shapeName=dyn_belt_r_01_capsule, center=<4 -0.0004 4.1254>, rotation=<rot 90 0 0>}]
 dyn_belt_l_03 [{shapeType=Capsule, length=11.8293, radius=4.6234, shapeName=dyn_belt_r_01_capsule, center=<4 -0.0004 4.1254>, rotation=<rot 90 0 0>}]
 dyn_belt_l_04 [{shapeType=Capsule, length=11.8293, radius=4.6234, shapeName=dyn_belt_r_01_capsule, center=<4 -0.0004 4.1254>, rotation=<rot 90 0 0>}]
 dyn_belt_r_01 [{shapeType=Capsule, length=11.8293, radius=4.6234, shapeName=dyn_belt_r_01_capsule, center=<4 -0.0004 4.1254>, rotation=<rot 90 0 0>}]
 dyn_belt_r_02 [{shapeType=Capsule, length=11.8293, radius=4.6234, shapeName=dyn_belt_r_01_capsule, center=<4 -0.0004 4.1254>, rotation=<rot 90 0 0>}]
 dyn_belt_r_03 [{shapeType=Capsule, length=11.8293, radius=4.6234, shapeName=dyn_belt_r_01_capsule, center=<4 -0.0004 4.1254>, rotation=<rot 90 0 0>}]
 dyn_belt_r_04 [{shapeType=Capsule, length=11.8293, radius=4.6234, shapeName=dyn_belt_r_01_capsule, center=<4 -0.0004 4.1254>, rotation=<rot 90 0 0>}]
 foot_l []
 foot_r []
 head [{shapeType=Sphere, radius=17.3125, center=<11.028 -0.174 0>}]
 lowerarm_l []
 lowerarm_r []
 neck_02 [{shapeType=Capsule, length=13.4375, radius=7.3175, center=<-0.2642 -0.3408 0>, rotation=<rot 90 -171.1193 -175.3978>}]
 pelvis [{shapeType=Capsule, length=8.3862, radius=10.9191, center=<-2.6883 -1.2679 0.0001>}]
 root [{shapeType=Box, extentX=327.5, extentY=327.5, extentZ=10, shapeName=root_box, center=<0 0 -2.3376>}]
 spine_01 [{shapeType=Capsule, length=10.9163, radius=8.3465, center=<1.6575 -1.7298 0>}]
 spine_02 [{shapeType=Capsule, length=9.397, radius=8.5439, center=<-1.6213 -3.2067 -0.0002>}]
 spine_03 [{shapeType=Capsule, length=10.4561, radius=6.9645, center=<0.0398 0.1711 0.0001>}]
 spine_04 [{shapeType=Capsule, length=14.6562, radius=13.6325, center=<2.1467 -2.0945 0.0002>}]
 spine_05 [
  {shapeType=Capsule, length=17.8437, radius=17.195, center=<-1.7514 1.3824 0.0001>}
  {shapeType=Capsule, length=12.7812, radius=17.0625, center=<0.7875 3.1143 0>}
 ]
 thigh_l []
 thigh_r []
 upperarm_l []
 upperarm_r []
}
constraints []
