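# Physics asset for the dynamic coat bones on the M_MED_Gimmick male medium player body cosmetic.
# Two front chains hang from spine_03 and three back chains from spine_01, with swing/twist-limited joints and a split center-back tail.
# Angular-free cross-links tie neighboring chains so panels move together.
# Leg, pelvis and spine capsules, plus a huge root sphere topping out at ground level, provide collision shapes alongside offset coat capsules and boxes.
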PhysicsAsset /BRCosmetics/Characters/Player/Male/Medium/Bodies/M_MED_Gimmick/Meshes/M_MED_Gimmick_Dyn_coat
bodies {
 calf_l [{shapeType=Capsule, length=47.1099, radius=16.6734, shapeName=pelvis_capsule, center=<-30.4611 6.2126 4.4558>, rotation=<rot 90 0 0>}]
 calf_r [{shapeType=Capsule, length=47.1099, radius=16.6734, shapeName=pelvis_capsule, center=<29.7555 -6.2124 -3.9891>, rotation=<rot 90 0 0>}]
 dyn_coat_bk_01 [{shapeType=Capsule, length=19.2053, radius=20, shapeName=dyn_coat_l_01_capsule, center=<8.7723 0.0007 20>, rotation=<rot 90 0 0>}]
 dyn_coat_bk_02 [
  {shapeType=Capsule, length=18.8008, radius=20, shapeName=dyn_coat_l_01_capsule, center=<8.8242 0.0006 20>, rotation=<rot 90 0 0>}
  {shapeType=Box, extentX=40, extentY=38, extentZ=18, shapeName=dyn_coat_bk_r_05_box, center=<9.5968 0.0006 8.5323>}
 ]
 dyn_coat_bk_04 [
  {shapeType=Capsule, length=30.5061, radius=20, shapeName=dyn_coat_l_01_capsule, center=<10.3356 0.0006 20>, rotation=<rot 90 0 0>}
  {shapeType=Box, extentX=40, extentY=40, extentZ=18, shapeName=dyn_coat_bk_r_05_box, center=<9.5968 0.0006 7.0203>}
 ]
 dyn_coat_bk_MSTR [{shapeType=Capsule, length=0.505, radius=0.505, shapeName=dyn_coat_l_MSTR_capsule}]
 dyn_coat_bk_l_01 [{shapeType=Capsule, length=25.8014, radius=20, shapeName=dyn_coat_l_01_capsule, center=<10.7079 0 20>, rotation=<rot 90 0 0>}]
 dyn_coat_bk_l_02 [{shapeType=Capsule, length=21.2561, radius=20, shapeName=dyn_coat_l_01_capsule, center=<10.3356 0 20>, rotation=<rot 90 0 0>}]
 dyn_coat_bk_l_04 [{shapeType=Capsule, length=21.2561, radius=20, shapeName=dyn_coat_l_01_capsule, center=<10.3356 0 20>, rotation=<rot 90 0 0>}]
 dyn_coat_bk_l_05 [
  {shapeType=Capsule, length=21.2561, radius=20, shapeName=dyn_coat_l_01_capsule, center=<10.3356 0 20>, rotation=<rot 90 0 0>}
  {
   shapeType=Box
   extentX=40
   extentY=37
   extentZ=18
   shapeName=dyn_coat_bk_r_05_box
   center=<9.5968 -0.6319 9.1152>
   rotation=<rot 0 0 -16.7847>
  }
 ]
 dyn_coat_bk_l_06 [{shapeType=Capsule, length=9.5061, radius=20, shapeName=dyn_coat_l_01_capsule, center=<6.0809 -0.0001 19.9994>, rotation=<rot 90 0 0>}]
 dyn_coat_bk_l_MSTR [{shapeType=Capsule, length=0.505, radius=0.505, shapeName=dyn_coat_l_MSTR_capsule}]
 dyn_coat_bk_r_01 [{shapeType=Capsule, length=25.8014, radius=20, shapeName=dyn_coat_l_01_capsule, center=<10.7079 0 20>, rotation=<rot 90 0 0>}]
 dyn_coat_bk_r_02 [{shapeType=Capsule, length=21.2561, radius=20, shapeName=dyn_coat_l_01_capsule, center=<10.3356 0 20>, rotation=<rot 90 0 0>}]
 dyn_coat_bk_r_04 [{shapeType=Capsule, length=21.2561, radius=20, shapeName=dyn_coat_l_01_capsule, center=<10.3356 0 20>, rotation=<rot 90 0 0>}]
 dyn_coat_bk_r_05 [
  {shapeType=Capsule, length=21.2561, radius=20, shapeName=dyn_coat_l_01_capsule, center=<10.3356 0 20>, rotation=<rot 90 0 0>}
  {shapeType=Box, extentX=40, extentY=37, extentZ=18, shapeName=dyn_coat_bk_r_05_box, center=<9.5968 0.7252 9.0335>, rotation=<rot 0 0 19.7942>}
 ]
 dyn_coat_bk_r_06 [{shapeType=Capsule, length=9.5061, radius=20, shapeName=dyn_coat_l_01_capsule, center=<6.0809 -0.0001 19.9994>, rotation=<rot 90 0 0>}]
 dyn_coat_bk_r_MSTR [{shapeType=Capsule, length=0.505, radius=0.505, shapeName=dyn_coat_l_MSTR_capsule}]
 dyn_coat_bk_split_l_01 [{shapeType=Capsule, length=16.6236, radius=20, shapeName=dyn_coat_l_01_capsule, center=<6.3429 0.0007 20>, rotation=<rot 90 0 0>}]
 dyn_coat_bk_split_l_02 [
  {shapeType=Capsule, length=7.8736, radius=20, shapeName=dyn_coat_l_01_capsule, center=<4.8323 0.0007 20.0005>, rotation=<rot 90 0 0>}
  {shapeType=Box, extentX=40, extentY=45, extentZ=18, shapeName=dyn_coat_bk_r_05_box, center=<9.5968 0.0006 7.0203>}
 ]
 dyn_coat_bk_split_r_01 [{shapeType=Capsule, length=16.6236, radius=20, shapeName=dyn_coat_l_01_capsule, center=<6.3429 0.0007 20>, rotation=<rot 90 0 0>}]
 dyn_coat_bk_split_r_02 [
  {shapeType=Capsule, length=7.8736, radius=20, shapeName=dyn_coat_l_01_capsule, center=<4.8323 0.0007 20.0005>, rotation=<rot 90 0 0>}
  {shapeType=Box, extentX=40, extentY=45, extentZ=18, shapeName=dyn_coat_bk_r_05_box, center=<9.5968 0.0006 7.0203>}
 ]
 dyn_coat_l_01 [{shapeType=Capsule, length=10.7561, radius=8, shapeName=dyn_coat_l_01_capsule, center=<4.667 -7.3674 1.8844>, rotation=<rot 90 0 0>}]
 dyn_coat_l_02 [{
   shapeType=Capsule
   length=12.2561
   radius=8
   shapeName=dyn_coat_l_01_capsule
   center=<5.44 -5.4574 5.0146>
   rotation=<rot 39.9131 -89.9998 -89.9998>
  }]
 dyn_coat_l_03 [{shapeType=Capsule, length=21.2561, radius=8, shapeName=dyn_coat_l_01_capsule, center=<10.3356 0 6.8531>, rotation=<rot 90 0 0>}]
 dyn_coat_l_04 [{shapeType=Capsule, length=21.2561, radius=15, shapeName=dyn_coat_l_01_capsule, center=<10.3356 0 14>, rotation=<rot 90 0 0>}]
 dyn_coat_l_05 [{shapeType=Capsule, length=11.0061, radius=15, shapeName=dyn_coat_l_01_capsule, center=<3.5998 0 14>, rotation=<rot 90 0 0>}]
 dyn_coat_l_06 [{shapeType=Capsule, length=18.0061, radius=15, shapeName=dyn_coat_l_01_capsule, center=<3.1822 0 14>, rotation=<rot 90 0 0>}]
 dyn_coat_l_MSTR [{shapeType=Capsule, length=0.505, radius=0.505, shapeName=dyn_coat_l_MSTR_capsule}]
 dyn_coat_r_01 [{shapeType=Capsule, length=10.7561, radius=8, shapeName=dyn_coat_l_01_capsule, center=<4.667 7.3674 1.8844>, rotation=<rot 90 0 0>}]
 dyn_coat_r_02 [{
   shapeType=Capsule
   length=12.2561
   radius=8
   shapeName=dyn_coat_l_01_capsule
   center=<5.44 4.2595 5.0146>
   rotation=<rot 39.9131 -89.9998 -89.9998>
  }]
 dyn_coat_r_03 [{shapeType=Capsule, length=21.2561, radius=8, shapeName=dyn_coat_l_01_capsule, center=<10.3356 0 6.8531>, rotation=<rot 90 0 0>}]
 dyn_coat_r_04 [{shapeType=Capsule, length=21.2561, radius=15, shapeName=dyn_coat_l_01_capsule, center=<10.3356 0 14>, rotation=<rot 90 0 0>}]
 dyn_coat_r_05 [{shapeType=Capsule, length=11.0061, radius=15, shapeName=dyn_coat_l_01_capsule, center=<3.5998 0 14>, rotation=<rot 90 0 0>}]
 dyn_coat_r_06 [{shapeType=Capsule, length=18.0061, radius=15, shapeName=dyn_coat_l_01_capsule, center=<3.1822 0 14>, rotation=<rot 90 0 0>}]
 dyn_coat_r_MSTR [{shapeType=Capsule, length=0.505, radius=0.505, shapeName=dyn_coat_l_MSTR_capsule}]
 pelvis [{shapeType=Capsule, length=6.967, radius=11.521, shapeName=pelvis_capsule, center=<-4.8171 -0.6518 -0.0005>}]
 root [{shapeType=Sphere, radius=500, shapeName=root_sphere, center=<0 0 -500>}]
 spine_01 [{shapeType=Capsule, length=6.4652, radius=10.9987, shapeName=pelvis_capsule, center=<-0.9991 -0.3721 0>}]
 spine_02 [{shapeType=Capsule, length=7.4401, radius=10.4987, shapeName=pelvis_capsule, center=<0.1451 -0.7547 0>}]
 spine_03 [{shapeType=Capsule, length=7.4401, radius=10.4987, shapeName=pelvis_capsule, center=<2.1769 0.258 0>}]
 thigh_l [{shapeType=Capsule, length=28.841, radius=11.8021, shapeName=pelvis_capsule, center=<-19.998 -0.2298 2.4342>, rotation=<rot 86.3502 180 180>}]
 thigh_r [{shapeType=Capsule, length=28.841, radius=11.8021, shapeName=pelvis_capsule, center=<19.998 0.2298 -2.4342>, rotation=<rot 86.3502 180 180>}]
}
constraints [
 {bone1Name=dyn_coat_l_01, bone2Name=dyn_coat_l_MSTR, swing1LimitDegrees=45, swing2LimitDegrees=45, twistLimitDegrees=5}
 {bone1Name=dyn_coat_l_02, bone2Name=dyn_coat_l_01, swing1LimitDegrees=25, swing2LimitDegrees=45, twistLimitDegrees=5}
 {bone1Name=dyn_coat_l_03, bone2Name=dyn_coat_l_02, swing1LimitDegrees=10, swing2LimitDegrees=45, twistLimitDegrees=5}
 {bone1Name=dyn_coat_l_04, bone2Name=dyn_coat_l_03, swing1LimitDegrees=10, swing2LimitDegrees=45, twistLimitDegrees=5}
 {bone1Name=dyn_coat_l_05, bone2Name=dyn_coat_l_04, swing1LimitDegrees=10, swing2LimitDegrees=45, twistLimitDegrees=5}
 {bone1Name=dyn_coat_l_06, bone2Name=dyn_coat_l_05, swing1LimitDegrees=10, swing2LimitDegrees=45, twistLimitDegrees=5}
 {bone1Name=dyn_coat_r_01, bone2Name=dyn_coat_r_MSTR, swing1LimitDegrees=45, swing2LimitDegrees=45, twistLimitDegrees=5}
 {bone1Name=dyn_coat_r_02, bone2Name=dyn_coat_r_01, swing1LimitDegrees=25, swing2LimitDegrees=45, twistLimitDegrees=5}
 {bone1Name=dyn_coat_r_03, bone2Name=dyn_coat_r_02, swing1LimitDegrees=10, swing2LimitDegrees=45, twistLimitDegrees=5}
 {bone1Name=dyn_coat_r_04, bone2Name=dyn_coat_r_03, swing1LimitDegrees=10, swing2LimitDegrees=45, twistLimitDegrees=5}
 {bone1Name=dyn_coat_r_05, bone2Name=dyn_coat_r_04, swing1LimitDegrees=10, swing2LimitDegrees=45, twistLimitDegrees=5}
 {bone1Name=dyn_coat_r_06, bone2Name=dyn_coat_r_05, swing1LimitDegrees=10, swing2LimitDegrees=45, twistLimitDegrees=5}
 {bone1Name=dyn_coat_bk_01, bone2Name=dyn_coat_bk_MSTR, swing1LimitDegrees=10, swing2LimitDegrees=45, twistLimitDegrees=5}
 {bone1Name=dyn_coat_bk_02, bone2Name=dyn_coat_bk_01, swing1LimitDegrees=15, swing2LimitDegrees=45, twistLimitDegrees=5}
 {bone1Name=dyn_coat_bk_split_l_01, bone2Name=dyn_coat_bk_04, swing1LimitDegrees=20, swing2LimitDegrees=45, twistLimitDegrees=5}
 {bone1Name=dyn_coat_bk_split_l_02, bone2Name=dyn_coat_bk_split_l_01, swing1LimitDegrees=20, swing2LimitDegrees=45, twistLimitDegrees=5}
 {bone1Name=dyn_coat_bk_split_r_01, bone2Name=dyn_coat_bk_04, swing1LimitDegrees=20, swing2LimitDegrees=45, twistLimitDegrees=5}
 {bone1Name=dyn_coat_bk_split_r_02, bone2Name=dyn_coat_bk_split_r_01, swing1LimitDegrees=20, swing2LimitDegrees=45, twistLimitDegrees=5}
 {bone1Name=dyn_coat_bk_l_01, bone2Name=dyn_coat_bk_l_MSTR, swing1LimitDegrees=10, swing2LimitDegrees=45, twistLimitDegrees=5}
 {bone1Name=dyn_coat_bk_l_02, bone2Name=dyn_coat_bk_l_01, swing1LimitDegrees=15, swing2LimitDegrees=45, twistLimitDegrees=5}
 {bone1Name=dyn_coat_bk_r_01, bone2Name=dyn_coat_bk_r_MSTR, swing1LimitDegrees=10, swing2LimitDegrees=45, twistLimitDegrees=5}
 {bone1Name=dyn_coat_bk_r_02, bone2Name=dyn_coat_bk_r_01, swing1LimitDegrees=15, swing2LimitDegrees=45, twistLimitDegrees=5}
 {
  bone1Name=calf_l
  bone2Name=thigh_l
  swing1LimitDegrees=45
  swing1Motion=Free
  swing2LimitDegrees=45
  swing2Motion=Free
  twistLimitDegrees=45
  twistMotion=Free
 }
 {
  bone1Name=dyn_coat_l_MSTR
  bone2Name=spine_03
  swing1LimitDegrees=45
  swing1Motion=Free
  swing2LimitDegrees=45
  swing2Motion=Free
  twistLimitDegrees=45
  twistMotion=Free
 }
 {
  bone1Name=dyn_coat_r_MSTR
  bone2Name=spine_03
  swing1LimitDegrees=45
  swing1Motion=Free
  swing2LimitDegrees=45
  swing2Motion=Free
  twistLimitDegrees=45
  twistMotion=Free
 }
 {
  bone1Name=dyn_coat_bk_MSTR
  bone2Name=spine_01
  swing1LimitDegrees=45
  swing1Motion=Free
  swing2LimitDegrees=45
  swing2Motion=Free
  twistLimitDegrees=45
  twistMotion=Free
 }
 {
  bone1Name=dyn_coat_bk_l_MSTR
  bone2Name=spine_01
  swing1LimitDegrees=45
  swing1Motion=Free
  swing2LimitDegrees=45
  swing2Motion=Free
  twistLimitDegrees=45
  twistMotion=Free
 }
 {
  bone1Name=dyn_coat_bk_r_MSTR
  bone2Name=spine_01
  swing1LimitDegrees=45
  swing1Motion=Free
  swing2LimitDegrees=45
  swing2Motion=Free
  twistLimitDegrees=45
  twistMotion=Free
 }
 {
  bone1Name=thigh_l
  bone2Name=pelvis
  swing1LimitDegrees=45
  swing1Motion=Free
  swing2LimitDegrees=45
  swing2Motion=Free
  twistLimitDegrees=45
  twistMotion=Free
 }
 {
  bone1Name=spine_01
  bone2Name=pelvis
  swing1LimitDegrees=45
  swing1Motion=Free
  swing2LimitDegrees=45
  swing2Motion=Free
  twistLimitDegrees=45
  twistMotion=Free
 }
 {
  bone1Name=spine_02
  bone2Name=spine_01
  swing1LimitDegrees=45
  swing1Motion=Free
  swing2LimitDegrees=45
  swing2Motion=Free
  twistLimitDegrees=45
  twistMotion=Free
 }
 {
  bone1Name=thigh_r
  bone2Name=pelvis
  swing1LimitDegrees=45
  swing1Motion=Free
  swing2LimitDegrees=45
  swing2Motion=Free
  twistLimitDegrees=45
  twistMotion=Free
 }
 {
  bone1Name=calf_r
  bone2Name=thigh_r
  swing1LimitDegrees=45
  swing1Motion=Free
  swing2LimitDegrees=45
  swing2Motion=Free
  twistLimitDegrees=45
  twistMotion=Free
 }
 {
  bone1Name=spine_03
  bone2Name=spine_02
  swing1LimitDegrees=45
  swing1Motion=Free
  swing2LimitDegrees=45
  swing2Motion=Free
  twistLimitDegrees=45
  twistMotion=Free
 }
 {bone1Name=dyn_coat_bk_r_05, bone2Name=dyn_coat_bk_r_04, swing1LimitDegrees=45, swing2LimitDegrees=45, twistLimitDegrees=45}
 {bone1Name=dyn_coat_bk_r_06, bone2Name=dyn_coat_bk_r_05, swing1LimitDegrees=45, swing2LimitDegrees=45, twistLimitDegrees=45}
 {bone1Name=dyn_coat_bk_l_05, bone2Name=dyn_coat_bk_l_04, swing1LimitDegrees=45, swing2LimitDegrees=45, twistLimitDegrees=45}
 {bone1Name=dyn_coat_bk_l_06, bone2Name=dyn_coat_bk_l_05, swing1LimitDegrees=45, swing2LimitDegrees=45, twistLimitDegrees=45}
 {
  bone1Name=pelvis
  bone2Name=root
  swing1LimitDegrees=45
  swing1Motion=Free
  swing2LimitDegrees=45
  swing2Motion=Free
  twistLimitDegrees=45
  twistMotion=Free
 }
 {bone1Name=dyn_coat_bk_r_04, bone2Name=dyn_coat_bk_r_02, swing1LimitDegrees=20, swing2LimitDegrees=45, twistLimitDegrees=5}
 {bone1Name=dyn_coat_bk_04, bone2Name=dyn_coat_bk_02, swing1LimitDegrees=20, swing2LimitDegrees=45, twistLimitDegrees=5}
 {bone1Name=dyn_coat_bk_l_04, bone2Name=dyn_coat_bk_l_02, swing1LimitDegrees=20, swing2LimitDegrees=45, twistLimitDegrees=5}
 {
  bone1Name=dyn_coat_r_05
  bone2Name=dyn_coat_bk_r_05
  swing1LimitDegrees=45
  swing1Motion=Free
  swing2LimitDegrees=45
  swing2Motion=Free
  twistLimitDegrees=45
  twistMotion=Free
 }
 {
  bone1Name=dyn_coat_l_05
  bone2Name=dyn_coat_bk_l_05
  swing1LimitDegrees=45
  swing1Motion=Free
  swing2LimitDegrees=45
  swing2Motion=Free
  twistLimitDegrees=45
  twistMotion=Free
 }
 {
  bone1Name=dyn_coat_bk_r_04
  bone2Name=dyn_coat_bk_04
  swing1LimitDegrees=45
  swing1Motion=Free
  swing2LimitDegrees=45
  swing2Motion=Free
  twistLimitDegrees=45
  twistMotion=Free
 }
 {
  bone1Name=dyn_coat_bk_l_04
  bone2Name=dyn_coat_bk_04
  swing1LimitDegrees=45
  swing1Motion=Free
  swing2LimitDegrees=45
  swing2Motion=Free
  twistLimitDegrees=45
  twistMotion=Free
 }
 {
  bone1Name=dyn_coat_r_04
  bone2Name=dyn_coat_bk_r_02
  swing1LimitDegrees=45
  swing1Motion=Free
  swing2LimitDegrees=45
  swing2Motion=Free
  twistLimitDegrees=45
  twistMotion=Free
 }
 {
  bone1Name=dyn_coat_l_04
  bone2Name=dyn_coat_bk_l_02
  swing1LimitDegrees=45
  swing1Motion=Free
  swing2LimitDegrees=45
  swing2Motion=Free
  twistLimitDegrees=45
  twistMotion=Free
 }
]
